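# MF_Rimv2 is a material function that adds a character rim light to the Emissive of incoming material attributes.
# It builds a contrasted Fresnel mask from AO and world normal Z, two directional rim lights tinted by time of day, and distance-based brightness from base color.
# Static switches add close-range behavior on the default quality path, while Epic blends separately.
# Collection parameters toggle the effect and scale distant brightness.
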MaterialFunction /BRCosmetics/Packages/Fortress_SharedMaterials/Base_Material_Functions/MF_Rimv2
graph {nodes [
  n1 = NamedRerouteUsage()
  n2 = Add(A=NamedRerouteUsage()) {constB=0.2}
  n3 = StaticSwitchParameter(True=Multiply(A=Saturate(None=@n2), B=@n1), False=@n1) {group=RimLight-Close, parameterName=Enable CloseRangeRimlight, sortPriority=0}
  n4 = NamedRerouteUsage()
  n5 = LinearInterpolate(
   A=@n4
   B=LinearInterpolate(A=NamedRerouteUsage(), B=@n4, Alpha=NamedRerouteUsage())
   Alpha=ScalarParameter() {defaultValue=1, group=RimLight, parameterName=Advanced_Rim_ThirdParty}
  )
  n6 = CollectionParameter() {
   collection=&FortniteMaterialParameters
   parameterName=bEnableCharacterRimLight
  }
  n7 = Multiply(A=Multiply(A=QualitySwitch(Default=@n3, Epic=@n5), B=@n6), B=ScalarParameter() {defaultValue=2, parameterName=Rimlight EyeExposure Boost})
  n8 = ScalarParameter() {defaultValue=1, parameterName=Rimlight_CvarBoost}
  n9 = Multiply(
   A=Multiply(A=ScalarParameter() {group=RimLight, parameterName=RimLight_Multiplier}, B=Multiply(A=@n7, B=@n8))
   B=ScalarParameter() {defaultValue=1, group=RimLight, parameterName=PerCharacterRimMultiplier}
  )
  n10 = Constant3Vector() {constant=<rgba 0.8021 0.7085 0 0>}
  n11 = GetMaterialAttributes(None=FunctionInput(Preview=SetMaterialAttributes!(MaterialAttributes, EmissiveColor=@n10) {attributeSetTypes [#1]}) {inputName=Material, inputType=FunctionInput_MaterialAttributes}) {attributeGetTypes [#2, #3, #1]}
  FunctionOutput(None=SetMaterialAttributes(MaterialAttributes=@n11.MaterialAttributes, EmissiveColor=Add(A=@n11.EmissiveColor, B=@n9)) {attributeSetTypes [#1]})
  MaterialFunctionCall() {materialFunction=/Game/Effects/Fort_Effects/Materials/_Functions/CurveColor_Time_of_Day}
  n12 = PreSkinnedPosition()
  n13 = Divide(
   A=FunctionInput(Preview=VertexInterpolator(VS=@n12).PS) {bUsePreviewValueAsDefault=true, inputName=Pre Skinned Local Position}
   B=Constant() {r=170}
  )
  NamedRerouteDeclaration(None=Saturate(None=ComponentMask(None=@n13) {b=true})) {name=Z-Gradient, nodeColor=<rgba 0 1 0.7891 1>}
  n14 = ScalarParameter() {defaultValue=2, group=RimLight-Close, parameterName=RimFresnelExp}
  n15 = ScalarParameter() {defaultValue=2, parameterName=FresEXP2}
  n16 = VertexNormalWS()
  n17 = Multiply(
   A=Fresnel(
    ExponentIn=StaticSwitchParameter(True=@n14, False=@n15) {group=RimLight-Close, parameterName=Enable CloseRangeRimlight, sortPriority=0}
    Normal=Transform(None=@n11.Normal)
   ) {baseReflectFraction=0}
   B=Multiply(
    A=FunctionInput() {inputName=AO, inputType=FunctionInput_Scalar, previewValue=<1 0 0 1>}
    B=MaterialFunctionCall(Float3=@n16) {materialFunction=&BreakOutFloat3Components}.B
   )
  )
  n18 = ScalarParameter() {defaultValue=0.03, group=RimLight, parameterName=RimContrast}
  NamedRerouteDeclaration(None=Saturate(None=MaterialFunctionCall(In=@n17, Contrast=@n18) {materialFunction=&CheapContrast}.Result)) {name=RimLightFresnel, nodeColor=<rgba 1 0.375 0 1>}
  n19 = ScalarParameter() {defaultValue=2, group=RimLight-Close, parameterName=Close Range Rim Strength, sortPriority=1}
  n20 = MaterialFunctionCall(
   Color_Morning=VectorParameter() {group=RimLight, parameterName=RimMorning, defaultValue=<rgba 0.875 0.8626 0.7474 1>}.RGB
   Color_Day=VectorParameter() {group=RimLight, parameterName=RimDay, defaultValue=<rgba 1.2 0.9092 0.5938 1>}.RGB
   Color_Evening=VectorParameter() {group=RimLight, parameterName=RimEvening, defaultValue=<rgba 0.9635 0.5427 0.2158 1>}.RGB
   Color_Night=VectorParameter() {group=RimLight, parameterName=RimNight, defaultValue=<rgba 0.3927 0.3462 0.651 1>}.RGB
  ) {materialFunction=&Color_Time_of_Day}
  n21 = StaticSwitchParameter(True=Multiply(A=@n19, B=@n20.Result), False=@n19) {group=RimLight-Close, parameterName=Enable TODCurveColorMult, sortPriority=0}
  n22 = Constant() {r=1}
  n23 = MaterialFunctionCall!(
   Blend Range=ScalarParameter() {defaultValue=800, desc=4500, parameterName=RimLightCloseRangeBlendDistance}
   Start Offset=ScalarParameter() {defaultValue=-2000, group=RimLight, parameterName=RimDistanceStartOffset}
  ) {materialFunction=/Engine/Functions/Engine_MaterialFunctions03/Particles/Distance_Blend}
  NamedRerouteDeclaration(None=StaticSwitchParameter(True=LinearInterpolate(A=@n21, B=@n22, Alpha=@n23.Result), False=@n23.Result) {group=RimLight-Close, parameterName=Enable CloseRangeRimlight, sortPriority=0}) {name=CloseRangeEmissive, nodeColor=<rgba 0 0.0625 1 1>}
  n24 = MaterialFunctionCall(
   X=ScalarParameter() {defaultValue=2, group=RimLight-Close, parameterName=Close Range Rim Brightness}
   Y=ScalarParameter() {defaultValue=4, group=RimLight, parameterName=Distance Rim Brightness}
  ) {materialFunction=&MakeFloat2}
  n25 = MaterialFunctionCall(Float2=QualitySwitch(Default=@n24.Result, Epic=Multiply(A=@n24.Result) {constB=0.75})) {materialFunction=&BreakOutFloat2Components}
  n26 = CollectionParameter() {
   collection=&FortniteMaterialParameters
   parameterName=DistanceRimBrightnessMultiply
  }
  n27 = Constant()
  n28 = Multiply(
   A=Clamp(None=@n11.BaseColor) {maxDefault=0.75, minDefault=0.25}
   B=LinearInterpolate(A=@n25.R, B=Multiply(A=@n25.G, B=@n26), Alpha=FunctionInput(Preview=@n27) {inputName=Distance Blend, inputType=FunctionInput_Scalar})
  )
  n29 = NamedRerouteUsage()
  NamedRerouteDeclaration(None=Multiply(A=LinearInterpolate(B=@n28, Alpha=@n29), B=NamedRerouteUsage())) {name=BrightnessAdjustments, nodeColor=<rgba 0.6641 0 1 1>}
  n30 = NamedRerouteUsage()
  n31 = FunctionInput(Preview=TextureSample() {samplerType=SAMPLERTYPE_Grayscale, texture=/Game/ArtTools/MaterialLibrary/Materials/T_MasterKey_Swatch_GMap}.G) {bUsePreviewValueAsDefault=true, inputName=Gmap, inputType=FunctionInput_Scalar}
  n32 = ScalarParameter() {defaultValue=2, group=RimLight, parameterName=RimMaskContrast}
  n33 = Saturate(None=Multiply(A=Power(Base=@n31, Exp=@n32), B=ScalarParameter() {defaultValue=1, group=RimLight, parameterName=RimMaskBoost}))
  n34 = Desaturation(None=Clamp(None=@n11.BaseColor) {maxDefault=0.7, minDefault=0.03}, Fraction=Constant() {r=0.6})
  n35 = MaterialFunctionCall!(
   BaseColor=@n34
   Gmap=@n33
   LightVector=VectorParameter() {group=RimLight, parameterName=RimLightDirection01, defaultValue=<rgba 0.524 1.631 2 1>}.RGB
  ) {materialFunction=/Game/Packages/Fortress_SharedMaterials/Base_Material_Functions/MF_Rimlight_SingleLight}
  n36 = MaterialFunctionCall(
   BaseColor=@n34
   Gmap=@n33
   LightVector=VectorParameter() {group=RimLight, parameterName=RimLightDirection02, defaultValue=<rgba 0.7413 0.657 1.5 1>}.RGB
  ) {materialFunction=&MF_Rimlight_SingleLight}
  n37 = Multiply(
   A=Multiply(A=@n20.Result, B=Add(A=@n35.Result, B=@n36.Result))
   B=ScalarParameter() {defaultValue=6, desc="artemis = 2.2 - Asteria 6(pending)", group=RimLight, parameterName=RimLightIntensity}
  )
  NamedRerouteDeclaration(None=Clamp(None=Multiply(A=@n30, B=@n37)) {maxDefault=500}) {name=RimLightColors, nodeColor=<rgba 0 1 0.1094 1>}
  PreSkinnedLocalBounds()
 ]}
